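# Physics asset for the F_MED_Bacteria female player body, set up for a cape and an upper cape flap.
# Capsules on the spine and neck provide body volumes, alongside small capsules on the two master cape joints.
# Each dyn cape joint has swing limits and zero twist against its master joint.
# Free angular constraints link the masters to spine_04 or neck_01 and chain spine_04 to spine_02, while spine_05 is unconstrained.
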PhysicsAsset /BRCosmetics/Characters/Player/Female/Medium/Bodies/F_MED_Bacteria/Meshes/F_MED_Bacteria_Physics
bodies {
 dyn_cape_jnt [{shapeType=Capsule, length=3.7628, radius=8.7347, shapeName=dyn_cape_jnt_capsule, center=<8.2196 -6.0084 0>}]
 dyn_upper_cape_flap_jnt [{shapeType=Capsule, length=3, radius=3, shapeName=dyn_upper_cape_flap_jnt_capsule, center=<4.5704 -0.9444 0>, rotation=<rot -90 180 180>}]
 master_cape_jnt [{shapeType=Capsule, length=0.505, radius=0.505, shapeName=master_cape_jnt_capsule, rotation=<rot 0 -90 0>}]
 master_upper_cape_flap_jnt [{shapeType=Capsule, length=0.505, radius=0.505, shapeName=master_upper_cape_flap_jnt_capsule, rotation=<rot 0 -90 0>}]
 neck_01 [{shapeType=Capsule, length=5.9324, radius=5.7525, shapeName=neck_01_capsule}]
 spine_02 [{
   shapeType=Capsule
   length=11.3275
   radius=8.5075
   shapeName=spine_04_capsule
   center=<-6.1239 -0.2893 -0.5992>
   rotation=<rot -11.3009 -90 -1.7159>
  }]
 spine_03 [{
   shapeType=Capsule
   length=5.3275
   radius=7.7575
   shapeName=spine_04_capsule
   center=<-0.3861 2.6349 0.1604>
   rotation=<rot -11.3009 -90.0001 -1.7159>
  }]
 spine_04 [{
   shapeType=Capsule
   length=12.0775
   radius=7.7575
   shapeName=spine_04_capsule
   center=<0.5632 0.3712 -0.263>
   rotation=<rot -11.3009 -90 -1.7159>
  }]
 spine_05 [{shapeType=Capsule, length=12.0157, radius=9.9165, shapeName=spine_05_capsule, center=<2.316 0.884 0>, rotation=<rot 0 9.9998 0>}]
}
constraints [
 {bone1Name=dyn_cape_jnt, bone2Name=master_cape_jnt, swing1LimitDegrees=75, swing2LimitDegrees=40, twistLimitDegrees=0}
 {
  bone1Name=master_cape_jnt
  bone2Name=spine_04
  swing1LimitDegrees=45
  swing1Motion=Free
  swing2LimitDegrees=45
  swing2Motion=Free
  twistLimitDegrees=45
  twistMotion=Free
 }
 {
  bone1Name=spine_04
  bone2Name=spine_03
  swing1LimitDegrees=45
  swing1Motion=Free
  swing2LimitDegrees=45
  swing2Motion=Free
  twistLimitDegrees=45
  twistMotion=Free
 }
 {
  bone1Name=spine_03
  bone2Name=spine_02
  swing1LimitDegrees=45
  swing1Motion=Free
  swing2LimitDegrees=45
  swing2Motion=Free
  twistLimitDegrees=45
  twistMotion=Free
 }
 {bone1Name=dyn_upper_cape_flap_jnt, bone2Name=master_upper_cape_flap_jnt, swing1LimitDegrees=60, swing2LimitDegrees=5, twistLimitDegrees=0}
 {
  bone1Name=master_upper_cape_flap_jnt
  bone2Name=neck_01
  swing1LimitDegrees=45
  swing1Motion=Free
  swing2LimitDegrees=45
  swing2Motion=Free
  twistLimitDegrees=45
  twistMotion=Free
 }
]
