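# Animation Blueprint for the female medium Bacteria body cosmetic, copying the parent mesh pose and adding secondary motion to cape and collar.
# RigidBody physics runs relative to spine_05 and switches off when backbling is worn, a flag refreshed on update and customization load.
# Additive bone offsets then adjust the cape for skydiving and the collar for targeting and front end.
# Updates also compute spineBendBack from clavicle rotation, feeding an unconnected cape-flap node.
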AnimBlueprint /BRCosmetics/Characters/Player/Female/Medium/Bodies/F_MED_Bacteria/Meshes/F_MED_Bacteria_AnimBP
target_skeleton=&/F_MED_Bacteria_Skeleton
graphs [
 AnimGraph {nodes [
   "Transform (Modify) Bone - Bone: master_upper_cape_flap_jnt"(Alpha=Get spineBendBack() -> (spineBendBack:double), Translation="0,-3.5,0", Rotation="0,0,0", Scale="1,1,1") {
    translationMode=BMM_Additive
    translationSpace=BCS_BoneSpace
    alphaBoolBlend {blendOption=Linear}
    boneToModify {boneName=master_upper_cape_flap_jnt}
    translation {y=-3.5}
   }
   n1 = RigidBody(
    ComponentPose=Local To Component(LocalPose=Copy Pose From Mesh() {bUseAttachedParent=true})
    Alpha=SelectFloat(A=0, B=1, bPickA=Get Is Wearing Backbling())
   ) {
    bClampLinearTranslationLimitToRefPose=true
    bOverrideWorldGravity=true
    overridePhysicsAsset=&/F_MED_Bacteria_Physics
    simulationSpace=BaseBoneSpace
    alphaBoolBlend {blendOption=Linear}
    baseBoneRef {boneName=spine_05}
    componentAppliedLinearAccClamp=<3000 3000 3000>
    componentLinearAccScale=<0.4 0.3 0.3>
    componentLinearVelScale=<0.4 0.3 0.3>
    overrideWorldGravity {z=-980}
    simSpaceSettings {maxAngularAcceleration=1000, maxAngularVelocity=3, maxLinearAcceleration=700, maxLinearVelocity=500, velocityScaleZ=0, worldAlpha=1}
   }
   n2 = "Transform (Modify) Bone - Bone: master_cape_jnt"!(ComponentPose=@n1, bAlphaBoolEnabled:bool=Get bIsSkyDiving(), Translation="0,0,0", Rotation="0,-15,0", Scale="1,1,1") -> (Pose) {
    alphaInputType=Bool
    rotationMode=BMM_Additive
    rotationSpace=BCS_BoneSpace
    alphaBoolBlend {blendOption=Linear}
    alphaScaleBiasClamp {bInterpResult=true}
    boneToModify {boneName=master_cape_jnt}
    rotation {yaw=-15}
   }
   n3 = "Transform (Modify) Bone - Bone: collar_helper_jnt"(ComponentPose=@n2, Alpha="To Float (Boolean)"(InBool=Get bIsTargeting()), Translation="0,-0.5,0", Rotation="0,0,0", Scale="0.8,0.8,0.8") {
    rotationSpace=BCS_WorldSpace
    scaleMode=BMM_Additive
    scaleSpace=BCS_BoneSpace
    translationMode=BMM_Additive
    translationSpace=BCS_BoneSpace
    alphaBoolBlend {blendOption=Linear}
    alphaScaleBiasClamp {bInterpResult=true, interpSpeedDecreasing=2, interpSpeedIncreasing=5}
    boneToModify {boneName=collar_helper_jnt}
    scale=<0.8 0.8 0.8>
    translation {y=-0.5}
   }
   n4 = "Transform (Modify) Bone - Bone: collar_helper_jnt"!(ComponentPose=@n3, bAlphaBoolEnabled:bool=Get bIsInFrontEnd(), Translation="0,0,1", Rotation="-5,0,30", Scale="1,1,1") -> (Pose) {
    alphaInputType=Bool
    rotationMode=BMM_Additive
    rotationSpace=BCS_BoneSpace
    scaleMode=BMM_Additive
    scaleSpace=BCS_BoneSpace
    translationMode=BMM_Additive
    translationSpace=BCS_BoneSpace
    alphaBoolBlend {blendOption=Linear}
    alphaScaleBiasClamp {interpSpeedDecreasing=2, interpSpeedIncreasing=5}
    boneToModify {boneName=collar_helper_jnt}
    rotation {pitch=-5, roll=30}
    translation {z=1}
   }
   Output Pose(Result=Component To Local(ComponentPose=@n4)) {layerGroup=None}
  ]}
 EventGraph {nodes [
   Event BlueprintUpdateAnimation()
   > n1 = Sequence()
   n2 = CheckForBackbling()
   Set Is Wearing Backbling(execute=[@n3.then, @n1.then_0], Is Wearing Backbling=OR Boolean(A=@n2, B=@n2.bIsBackblingLong))
   Set spineBendBack(
    execute=@n1.then_1
    spineBendBack:double=GetAlphaFromReferenceTransformDelta(
     SkeletalMeshComponent=GetOwnerSkeletalMeshComponent()
     BoneName=clavicle_l
     BaseName=spine_05
     DeltaComponentType=EulerY
     InputRange=(X=11,Y=42)
     TwistAxis="1,0,0"
    )
   ) -> (Output_Get:double)
   n3 = Event OnCharacterCustomizationLoaded()
  ]}
]
variables [
 Is Wearing Backbling {type=boolean}
 spineBendBack {type=number}
]
