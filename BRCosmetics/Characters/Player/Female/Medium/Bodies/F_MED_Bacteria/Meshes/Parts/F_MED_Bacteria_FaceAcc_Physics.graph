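# Physics asset for the face accessory part of the F_MED_Bacteria player cosmetic, simulating two dynamic ear joints.
# It has capsule bodies on the left ear, right ear and head bones.
# Each ear is constrained to the head with equal, narrow swing limits and a zero twist limit, so the ears can sway slightly around the head.
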PhysicsAsset /BRCosmetics/Characters/Player/Female/Medium/Bodies/F_MED_Bacteria/Meshes/Parts/F_MED_Bacteria_FaceAcc_Physics
bodies {
 dyn_L_ear_jnt [{
   shapeType=Capsule
   length=7.6296
   radius=3.5327
   shapeName=dyn_L_ear_jnt_capsule
   center=<7.2615 0.1211 -0.4256>
   rotation=<rot 0 93.228 -89.1392>
  }]
 dyn_R_ear_jnt [{
   shapeType=Capsule
   length=7.6296
   radius=3.5327
   shapeName=dyn_R_ear_jnt_capsule
   center=<-6.8057 -0.0954 0.4328>
   rotation=<rot 0 93.228 -89.1392>
  }]
 head [{shapeType=Capsule, length=5, radius=5, shapeName=head_capsule, center=<1.611 0 0>, rotation=<rot -90 27.9385 -27.9389>}]
}
constraints [
 {bone1Name=dyn_L_ear_jnt, bone2Name=head, swing1LimitDegrees=25, swing2LimitDegrees=25, twistLimitDegrees=0}
 {bone1Name=dyn_R_ear_jnt, bone2Name=head, swing1LimitDegrees=25, swing2LimitDegrees=25, twistLimitDegrees=0}
]
